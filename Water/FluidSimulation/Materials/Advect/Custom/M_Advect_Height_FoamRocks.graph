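# Unlit water fluid-sim update material that advects height and foam, adding fluid and foam around terrain and rocks, packing fluid height, ground height and foam into RGB.
# It back-traces HeightRT along averaged staggered VelocityRT, applying evaporation and ForcesRT.
# Wave height against WaterDistanceField capture drives injection, while exponential damping eases height toward resting depth.
# Foam is blurred, faded and fed by edges and speed; switches select capture, terrain and foam.
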Material /Water/FluidSimulation/Materials/Advect/Custom/M_Advect_Height_FoamRocks
graph {
 properties {bAllowNegativeEmissiveColor=true, pixelDepthOffsetMode=PDOM_Legacy, shadingModel=MSM_Unlit, usedShadingModels=Unlit}
 nodes [
  Constant() {r=-50}
  n1 = TextureCoordinate()
  LinearInterpolate(
   B=ScalarParameter() {defaultValue=0.2, parameterName=Foam Fading Max}
   Alpha=TextureSample(UVs=Multiply(A=@n1) {constB=2}) {texture=/Water/Textures/Foam/T_WaterFlow_01_Foam_Tiled}.R
  )
  n2 = TextureSampleParameter2D!(
   UVs=MaterialFunctionCall() {materialFunction=/FortWater/FluidSimulation/Materials/Functions/PointSampleFluidSimUVs}.Result
   Level
   Apply View MipBias
  ) {constMipValue=0, mipValueMode=TMVM_MipLevel, parameterName=HeightRT, texture=&Black}
  n3 = MaterialFunctionCall!(Position=MaterialFunctionCall!(UV) {materialFunction=/Water/FluidSimulation/Materials/Functions/Fluidsim_RenderTargetWorldPosition}.Fluidsim WorldPosition) {materialFunction=/Water/Materials/Functions/WaterBodyData}
  n4 = MaterialFunctionCall(Position=MaterialFunctionCall() {materialFunction=&Fluidsim_RenderTargetWorldPosition}.Fluidsim WorldPosition) {materialFunction=&WaterBodyData}
  n5 = CollectionParameter() {collection=/Water/Materials/MPC/MPC_Water, parameterName=WaveCaptureSize}
  n6 = CollectionParameter() {collection=&MPC_Water, parameterName=FluidSimSize}
  n7 = TextureSample(UVs=MaterialFunctionCall(Texture Scale=Divide(A=@n5, B=@n6) {desc=Size of Wave Capture vs Sim area}) {materialFunction=&ScaleUVsByCenter}.UVs) {desc="B = Wave Z", samplerType=SAMPLERTYPE_LinearColor, texture=/Water/FluidSimulation/WaveFoam/FoamRT}
  n8 = ScalarParameter() {defaultValue=1, parameterName=Wave Z Multiplier}
  n9 = TextureSample() {samplerType=SAMPLERTYPE_LinearColor, texture=/Water/FluidSimulation/WaveFoam/WaterDistanceField}
  n10 = Subtract(A=Add(A=@n4.Water Z Height, B=Multiply(A=@n7.B, B=@n8)), B=@n9.G) {desc=Height Over Terrain/Rocks}
  n11 = Divide(A=Abs(None=@n10), B=ScalarParameter() {defaultValue=6, parameterName=Foam Band})
  n12 = Multiply(A=Subtract(A=@n4.Water Z Height, B=@n9.G) {desc="Water Z - Capture Z = Water Depth"}) {constB=-1}
  n13 = ScalarParameter() {defaultValue=16, parameterName=Foam Detection Band}
  n14 = Multiply(A=OneMinus(None=Saturate(None=@n11)), B=Saturate(None=Divide(A=@n12, B=@n13)))
  n15 = ScalarParameter() {defaultValue=75, desc=75, parameterName=Fluid Addition Above Rocks}
  n16 = CollectionParameter() {collection=&MPC_Water, parameterName=DeltaTime}
  n17 = MaterialFunctionCall() {materialFunction=&PointSampleFluidSimUVs}
  n18 = TextureSampleParameter2D!(UVs=@n17.Result, Apply View MipBias) {parameterName=VelocityRT, texture=&Black}
  n19 = Divide(B=CollectionParameter() {collection=&MPC_Water, parameterName=FluidSimResolution})
  n20 = Constant2Vector() {r=1}
  n21 = TextureSampleParameter2D(UVs=Add(A=@n17.Result, B=Multiply(A=@n19, B=@n20))) {parameterName=VelocityRT, texture=&Black}
  n22 = TextureSampleParameter2D(UVs=@n17.Result) {parameterName=VelocityRT, texture=&Black}
  n23 = Constant2Vector() {g=1}
  n24 = TextureSampleParameter2D(UVs=Add(A=@n17.Result, B=Multiply(A=@n19, B=@n23))) {parameterName=VelocityRT, texture=&Black}
  n25 = AppendVector(A=Divide(A=Add(A=@n18.R, B=@n21.R)), B=Divide(A=Add(A=@n22.G, B=@n24.G)))
  n26 = Subtract(
   A=MaterialFunctionCall() {materialFunction=&PointSampleFluidSimUVs}.Result
   B=Divide(A=Multiply(A=@n16, B=@n25), B=CollectionParameter() {collection=&MPC_Water, parameterName=FluidSimSize})
  )
  n27 = TextureSampleParameter2D!(UVs=@n26, Level, Apply View MipBias) {constMipValue=0, mipValueMode=TMVM_MipLevel, parameterName=HeightRT, texture=&Black}
  n28 = ScalarParameter() {defaultValue=0.125, parameterName=Evaporation}
  n29 = CollectionParameter() {collection=&MPC_Water, parameterName=DeltaTime}
  n30 = TextureSampleParameter2D(UVs=MaterialFunctionCall() {materialFunction=&PointSampleFluidSimUVs}.Result) {constMipValue=0, mipValueMode=TMVM_MipLevel, parameterName=ForcesRT, texture=&Black}
  n31 = Add(A=Subtract(A=@n27.R, B=Multiply(A=@n28, B=@n29)), B=@n30.R) {constB=0}
  n32 = Multiply(A=@n10, B=ScalarParameter() {defaultValue=0.6, desc=0.6, parameterName=Fluid Per Frame})
  n33 = Multiply(A=Multiply(A=@n9.R, B=CollectionParameter() {collection=&MPC_Water, parameterName=WaveCaptureSize})) {constB=-1}
  n34 = ScalarParameter() {defaultValue=-64, parameterName=Water DF Bias}
  n35 = Saturate(None=Divide(A=Subtract(A=@n33, B=@n34), B=ScalarParameter() {defaultValue=64, parameterName=Water DF Length}))
  n36 = Add(
   A=Multiply(A=Multiply(A=@n14, B=@n15), B=CollectionParameter() {collection=&MPC_Water, parameterName=DeltaTime})
   B=LinearInterpolate(A=@n31, B=Max(A=@n32, B=@n31), Alpha=Multiply(A=@n35))
  )
  n37 = StaticSwitchParameter(
   True=If(A=@n2.G, B=Constant() {r=-65000}, "A > B"=@n36, "A == B"=@n36, "A < B"=Max(A=@n3.Water Depth) {constB=0}) {equalsThreshold=0.5}
   False=@n36
  ) {parameterName=Use Terrain System}
  n38 = TextureSample() {samplerType=SAMPLERTYPE_LinearColor, texture=&WaterDistanceField}
  n39 = Max(A=Subtract(A=@n3.Water Z Height, B=@n38.G) {constA=0}) {constB=0}
  n40 = If(A=@n2.G, B=Constant() {r=-65000}, "A > B"=@n37, "A == B"=@n37, "A < B"=@n39) {equalsThreshold=0.5}
  MaterialExpressionScalarParameter_7 = ScalarParameter() {defaultValue=1, parameterName=Fading Back}
  MaterialExpressionCollectionParameter_6 = CollectionParameter() {collection=&MPC_Water, parameterName=DeltaTime}
  n41 = Custom!(x=@MaterialExpressionScalarParameter_7, t=@MaterialExpressionCollectionParameter_6) {
   code="1.0-exp(-x*t);"
   description=Exp Damping
   outputType=CMOT_Float1
   inputs [
    {inputName=x, input {expression=&self:MaterialExpressionScalarParameter_7}}
    {
     inputName=t
     input {expression=&self:MaterialExpressionCollectionParameter_6}
    }
   ]
  }
  n42 = LinearInterpolate(
   A=@n3.Terrain Z Height
   B=Multiply(A=@n3.Water Depth) {constB=-1, desc=Water as Flat Plane Mode}
   Alpha=ScalarParameter() {parameterName=TerrainUseWaterDepthOnly}
  )
  n43 = Constant()
  n44 = Saturate(None=OneMinus(None=ScalarParameter() {defaultValue=0.05, parameterName="Water Fe[th Foam Fade"}))
  n45 = Saturate(None=@n2.R)
  n46 = MaterialFunctionCall!(OffsetAmount=Divide(B=CollectionParameter() {collection=&MPC_Water, parameterName=FluidSimResolution})) {materialFunction=/Engine/Functions/Engine_MaterialFunctions01/ImageAdjustment/BlurSampleOffsets}
  n47 = Add(
   A=TextureSampleParameter2D(UVs=Add(A=@n46."0,1", B=@n26)) {constMipValue=0, mipValueMode=TMVM_MipLevel, parameterName=HeightRT, texture=&Black}.B
   B=TextureSampleParameter2D(UVs=Add(A=@n46."0,-1", B=@n26)) {constMipValue=0, mipValueMode=TMVM_MipLevel, parameterName=HeightRT, texture=&Black}.B
  )
  n48 = TextureSampleParameter2D(UVs=Add(A=@n46."1,0", B=@n26)) {constMipValue=0, mipValueMode=TMVM_MipLevel, parameterName=HeightRT, texture=&Black}
  n49 = TextureSampleParameter2D(UVs=Add(A=@n46."-1,0", B=@n26)) {constMipValue=0, mipValueMode=TMVM_MipLevel, parameterName=HeightRT, texture=&Black}
  n50 = Divide(A=Add(A=Add(A=@n47, B=@n48.B), B=Add(A=@n49.B, B=@n27.B))) {constB=5}
  n51 = Multiply(
   A=ScalarParameter() {defaultValue=1, parameterName=Foam Fading}
   B=CollectionParameter() {collection=&MPC_Water, parameterName=DeltaTime}
  )
  n52 = Multiply(A=Multiply(A=Add(A=@n50, B=@n30.G), B=LinearInterpolate(A=@n44, Alpha=@n45)), B=Saturate(None=OneMinus(None=@n51)))
  n53 = TextureSampleParameter2D!(UVs, Apply View MipBias) {parameterName=VelocityRT, texture=&Black}
  n54 = Divide(
   A=MaterialFunctionCall(Vector 2=ComponentMask(None=@n53.RGB) {g=true, r=true}) {materialFunction=/Engine/Functions/Engine_MaterialFunctions02/Utility/VectorLength}.V2 Length
   B=ScalarParameter() {defaultValue=1024, parameterName=Velocity for Foam}
  )
  n55 = Add(
   A=Multiply(A=@n14, B=ScalarParameter() {defaultValue=0.1, desc=0.01, parameterName=Edge Foam})
   B=Multiply(A=Saturate(None=@n54), B=ScalarParameter() {defaultValue=0.02, desc=0.02, parameterName=Foam Injection})
  )
  MaterialFunctionCall(
   X=StaticSwitchParameter(True=LinearInterpolate(A=@n40, B=@n39, Alpha=@n41), False=@n37) {defaultValue=true, parameterName=Use Scene Capture}
   Y=StaticSwitchParameter(True=@n38.G, False=StaticSwitchParameter(True=@n42, False=@n43) {parameterName=Use Terrain System}) {defaultValue=true, parameterName=Use Scene Capture}
   Z=StaticSwitchParameter(True=Add(A=@n52, B=@n55), False=Constant()) {defaultValue=true, parameterName=Enable Foam}
  ) {materialFunction=&MakeFloat3}
 ]
}
