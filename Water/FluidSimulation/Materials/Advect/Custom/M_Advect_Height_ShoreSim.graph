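# Unlit shore fluid-sim pass that advects the height render target, packing water depth, ground height and foam into three channels.
# It backtraces along averaged staggered velocities, seeding cells with invalid ground height from water body or scene-capture data.
# Depth gains forces and wave-crash water from FoamRT, relaxes exponentially toward rest depth and evaporates.
# Foam is injected by velocity within a depth mask, then advected and faded.
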Material /Water/FluidSimulation/Materials/Advect/Custom/M_Advect_Height_ShoreSim
graph {
 properties {bAllowNegativeEmissiveColor=true, pixelDepthOffsetMode=PDOM_Legacy, shadingModel=MSM_Unlit, usedShadingModels=Unlit}
 nodes [
  n1 = MaterialFunctionCall() {materialFunction=/FortWater/FluidSimulation/Materials/Functions/PointSampleFluidSimUVs}
  n2 = MaterialFunctionCall!(Position=MaterialFunctionCall!(UV) {materialFunction=/Water/FluidSimulation/Materials/Functions/Fluidsim_RenderTargetWorldPosition}.Fluidsim WorldPosition) {materialFunction=/Water/Materials/Functions/WaterBodyData}
  n3 = Add(A=TextureSample() {samplerType=SAMPLERTYPE_LinearColor, texture=/Water/FluidSimulation/WaveFoam/FoamRT}.B, B=@n2.Water Z Height) {desc=Wave + Water elevation}
  n4 = Max(A=Subtract(
    A=Subtract(A=@n3, B=@n2.Terrain Z Height) {desc=Height Above Terrain}
    B=ScalarParameter() {defaultValue=24, parameterName=Wave Crash Threshold}
   )) {constB=0}
  n5 = SphereMask(
   A=@n2.Water Depth
   B=ScalarParameter() {defaultValue=128, parameterName=Crash Depth}
   Radius=ScalarParameter() {defaultValue=64, parameterName=Wave Radius}
   Hardness=Constant() {r=0.25}
  )
  n6 = Multiply(
   A=Multiply(A=Multiply(A=@n4, B=@n5), B=CollectionParameter() {collection=/Water/Materials/MPC/MPC_Water, parameterName=DeltaTime})
   B=ScalarParameter() {defaultValue=0.5, parameterName=Wave Fluid Addition}
  )
  n7 = TextureSampleParameter2D!(
   UVs=MaterialFunctionCall() {materialFunction=&PointSampleFluidSimUVs}.Result
   Level
   Apply View MipBias
  ) {constMipValue=0, mipValueMode=TMVM_MipLevel, parameterName=HeightRT, texture=&Black}
  n8 = MaterialFunctionCall(Position=MaterialFunctionCall() {materialFunction=&Fluidsim_RenderTargetWorldPosition}.Fluidsim WorldPosition) {materialFunction=&WaterBodyData}
  n9 = Max(A=@n8.Water Depth) {constB=0}
  n10 = MaterialFunctionCall() {materialFunction=&PointSampleFluidSimUVs}
  n11 = Divide(B=CollectionParameter() {collection=&MPC_Water, parameterName=FluidSimResolution})
  n12 = Multiply(A=@n11, B=Constant2Vector() {r=1})
  n13 = Add(
   A=TextureSampleParameter2D!(UVs=@n10.Result, Apply View MipBias) {parameterName=VelocityRT, texture=&Black}.R
   B=TextureSampleParameter2D(UVs=Add(A=@n10.Result, B=@n12)) {parameterName=VelocityRT, texture=&Black}.R
  )
  n14 = Multiply(A=@n11, B=Constant2Vector() {g=1})
  n15 = Add(
   A=TextureSampleParameter2D(UVs=@n10.Result) {parameterName=VelocityRT, texture=&Black}.G
   B=TextureSampleParameter2D(UVs=Add(A=@n10.Result, B=@n14)) {parameterName=VelocityRT, texture=&Black}.G
  )
  n16 = Multiply(
   A=CollectionParameter() {collection=&MPC_Water, parameterName=DeltaTime}
   B=AppendVector(A=Divide(A=@n13), B=Divide(A=@n15))
  )
  n17 = CollectionParameter() {collection=&MPC_Water, parameterName=FluidSimSize}
  n18 = TextureSampleParameter2D!(
   UVs=Subtract(
    A=MaterialFunctionCall() {materialFunction=&PointSampleFluidSimUVs}.Result
    B=Divide(A=@n16, B=@n17)
   )
   Level
   Apply View MipBias
  ) {constMipValue=0, mipValueMode=TMVM_MipLevel, parameterName=HeightRT, texture=&Black}
  n19 = StaticSwitchParameter(True=If(A=@n7.G, B=Constant() {r=-65000}, "A > B"=@n18.R, "A == B"=@n18.R, "A < B"=@n9) {equalsThreshold=0.5}, False=@n18.R) {defaultValue=true, parameterName=Use Terrain System}
  n20 = TextureSample() {desc=Scene Capture DF and Depth, samplerType=SAMPLERTYPE_LinearColor, texture=/Water/FluidSimulation/WaveFoam/WaterDistanceField}
  n21 = Subtract(A=@n8.Water Z Height, B=@n20.G) {constA=0}
  n22 = StaticSwitchParameter(True=If(A=@n7.G, B=Constant() {r=-65000}, "A > B"=@n19, "A == B"=@n19, "A < B"=Max(A=@n21) {constB=0}) {equalsThreshold=0.5}, False=@n19) {parameterName=Use Scene Capture}
  MaterialExpressionScalarParameter_7 = ScalarParameter() {defaultValue=0.2, parameterName=Fading Back}
  MaterialExpressionCollectionParameter_3 = CollectionParameter() {collection=&MPC_Water, parameterName=DeltaTime}
  n23 = Custom!(x=@MaterialExpressionScalarParameter_7, t=@MaterialExpressionCollectionParameter_3) {
   code="1.0-exp(-x*t);"
   description=Exp Damping
   outputType=CMOT_Float1
   inputs [
    {inputName=x, input {expression=&self:MaterialExpressionScalarParameter_7}}
    {
     inputName=t
     input {expression=&self:MaterialExpressionCollectionParameter_3}
    }
   ]
  }
  n24 = ScalarParameter() {defaultValue=0.25, parameterName=Evaporation}
  n25 = CollectionParameter() {collection=&MPC_Water, parameterName=DeltaTime}
  n26 = Max(A=Subtract(A=LinearInterpolate(A=@n22, B=@n9, Alpha=@n23), B=Multiply(A=@n24, B=@n25))) {constB=0}
  n27 = LinearInterpolate(
   A=@n8.Terrain Z Height
   B=Multiply(A=@n8.Water Depth) {constB=-1, desc=Water as Flat Plane Mode}
   Alpha=ScalarParameter() {parameterName=TerrainUseWaterDepthOnly}
  )
  n28 = Constant()
  n29 = MaterialFunctionCall(Vector 2=ComponentMask(None=TextureSampleParameter2D!(UVs, Apply View MipBias) {parameterName=VelocityRT, texture=&Black}.RGB) {g=true, r=true}) {materialFunction=/Engine/Functions/Engine_MaterialFunctions02/Utility/VectorLength}
  n30 = ScalarParameter() {defaultValue=768, parameterName=Velocity for Foam}
  n31 = Multiply(
   A=SphereMask(
    A=@n2.Water Depth
    B=ScalarParameter() {parameterName=Foam Depth}
    Radius=ScalarParameter() {defaultValue=128, parameterName=Foam Radius}
    Hardness=Constant() {r=0.25}
   )
   B=Saturate(None=Divide(A=@n29.V2 Length, B=@n30))
  )
  n32 = ScalarParameter() {defaultValue=0.1, parameterName=Foam Injection}
  n33 = ScalarParameter() {defaultValue=0.4, parameterName=Foam Fading}
  n34 = CollectionParameter() {collection=&MPC_Water, parameterName=DeltaTime}
  n35 = Saturate(None=OneMinus(None=Multiply(A=@n33, B=@n34)))
  MaterialFunctionCall(
   X=Add(
    A=TextureSampleParameter2D!(UVs=@n1.Result, Level, Apply View MipBias) {constMipValue=0, mipValueMode=TMVM_MipLevel, parameterName=ForcesRT, texture=&Black}.R
    B=Add(A=@n6, B=@n26)
   ) {constB=0}
   Y=StaticSwitchParameter(True=@n20.G, False=StaticSwitchParameter(True=@n27, False=@n28) {defaultValue=true, parameterName=Use Terrain System}) {parameterName=Use Scene Capture}
   Z=Add(A=Multiply(A=@n31, B=@n32), B=Multiply(A=@n18.B, B=@n35))
  ) {materialFunction=&MakeFloat3}
 ]
}
