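# Unlit material that advects the shore fluid simulation's velocity render target over one timestep.
# Red and green velocities are semi-Lagrangian back-traced separately, pairing components from diagonally half-texel-offset samples as on a staggered grid, using MPC_Water timestep, size and resolution.
# The recombined result feeds two separate branches: one adds VelocityForces, the other fades velocity out as FoamRT wave height (blue) exceeds a threshold.
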Material /Water/FluidSimulation/Materials/Advect/Custom/M_Advect_Velocity_ShoreSim
graph {
 properties {bAllowNegativeEmissiveColor=true, pixelDepthOffsetMode=PDOM_Legacy, shadingModel=MSM_Unlit, usedShadingModels=Unlit}
 nodes [
  n1 = TextureSampleParameter2D() {
   parameterName=VelocityForces
   samplerType=SAMPLERTYPE_LinearColor
   texture=/Engine/Functions/Engine_MaterialFunctions02/PivotPainter2/Black_1x1_EXR_Texture
  }
  n2 = MaterialFunctionCall() {materialFunction=/FortWater/FluidSimulation/Materials/Functions/PointSampleFluidSimUVs}
  n3 = TextureSampleParameter2D!(
   UVs=MaterialFunctionCall() {materialFunction=&PointSampleFluidSimUVs}.Result
   Level=Constant()
   Apply View MipBias
  ) {constMipValue=0, mipValueMode=TMVM_MipLevel, parameterName=VelocityRT, texture=&Black}
  n4 = Multiply(A=Divide(B=CollectionParameter() {collection=/Water/Materials/MPC/MPC_Water, parameterName=FluidSimResolution})) {constB=0.5}
  n5 = Constant2Vector() {g=1, r=-1}
  n6 = TextureSampleParameter2D(
   UVs=Add(
    A=MaterialFunctionCall() {materialFunction=&PointSampleFluidSimUVs}.Result
    B=Multiply(A=@n4, B=@n5)
   )
   Level=Constant()
  ) {constMipValue=0, mipValueMode=TMVM_MipLevel, parameterName=VelocityRT, texture=&Black}
  n7 = CollectionParameter() {collection=&MPC_Water, parameterName=DeltaTime}
  n8 = CollectionParameter() {collection=&MPC_Water, parameterName=FluidSimSize}
  n9 = Divide(A=Multiply(A=AppendVector(A=@n3.R, B=@n6.G), B=@n7), B=@n8)
  n10 = MaterialFunctionCall() {materialFunction=&PointSampleFluidSimUVs}
  n11 = Constant2Vector() {g=-1, r=1}
  n12 = TextureSampleParameter2D(
   UVs=Add(
    A=MaterialFunctionCall() {materialFunction=&PointSampleFluidSimUVs}.Result
    B=Multiply(A=@n4, B=@n11)
   )
   Level=Constant()
  ) {constMipValue=0, mipValueMode=TMVM_MipLevel, parameterName=VelocityRT, texture=&Black}
  n13 = Divide(A=Multiply(A=@n7, B=AppendVector(A=@n12.R, B=@n3.G)), B=@n8)
  n14 = MaterialFunctionCall(
   X=TextureSampleParameter2D(UVs=Subtract(A=@n2.Result, B=@n9), Level=Constant()) {desc=New, mipValueMode=TMVM_MipLevel, parameterName=VelocityRT, texture=&Black}.R
   Y=TextureSampleParameter2D(UVs=Subtract(A=@n10.Result, B=@n13), Level=Constant()) {desc=New, mipValueMode=TMVM_MipLevel, parameterName=VelocityRT, texture=&Black}.G
  ) {materialFunction=&MakeFloat3}
  Add(A=@n1.RGB, B=@n14.Result)
  n15 = Divide(
   A=Subtract(
    A=TextureSample() {desc="B = Wave Z", samplerType=SAMPLERTYPE_LinearColor, texture=/Water/FluidSimulation/WaveFoam/FoamRT}.B
    B=ScalarParameter() {defaultValue=16, parameterName=Min Wave to Erase Velocity}
   )
   B=ScalarParameter() {defaultValue=16, parameterName=Velocity fade Length}
  )
  Multiply(A=@n14.Result, B=OneMinus(None=Saturate(None=@n15)))
 ]
}
